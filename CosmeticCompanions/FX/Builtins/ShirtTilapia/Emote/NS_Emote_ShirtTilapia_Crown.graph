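# Niagara effect system for the ShirtTilapia cosmetic companion's Crown emote.
# It holds three emitters named for a glow, crown stars and head stars, with the LWE prefix suggesting lightweight emitters.
# SystemState is the only listed module, and no renderers, user variables, data interfaces or dynamic inputs appear, so the emitters' detailed behavior is not shown.
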NiagaraSystem /CosmeticCompanions/FX/Builtins/ShirtTilapia/Emote/NS_Emote_ShirtTilapia_Crown
dependencies {usedDataInterfaces [], usedDynamicInputs [], usedModules [&SystemState]}
emitters {
 LWE_Glow {rendererClasses [], renderers []}
 LWE_Stars_Crown {rendererClasses [], renderers []}
 LWE_Stars_Head {rendererClasses [], renderers []}
}
summary {
 systemName=NS_Emote_ShirtTilapia_Crown
 emitters [
  {emitterName=LWE_Stars_Crown, rendererClasses []}
  {emitterName=LWE_Stars_Head, rendererClasses []}
  {emitterName=LWE_Glow, rendererClasses []}
 ]
 userVariables []
}
