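# Niagara trail effect for the ShirtTilapia cosmetic companion emote, built from two sprite spark emitters and six LWE In, Out and Start emitters whose contents are not shown.
# The spark emitters start at staggered delays, burst and stream particles from offset boxes, and launch them along negative X with drag.
# Sparks stretch with speed, shrink over their short life, drive a dynamic material parameter, and fade with alpha scaled by time of day.
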NiagaraSystem /CosmeticCompanions/FX/Builtins/ShirtTilapia/Emote/NS_Emote_ShirtTilapia_Trail
dependencies {
 usedDataInterfaces [&NiagaraDataInterfaceCurve, &NiagaraDataInterfaceVector2DCurve]
 usedDynamicInputs [&SimulationPosition]
 usedModules [&SystemState]
}
emitters {
 LWE_Trail_In_01 {rendererClasses [], renderers []}
 LWE_Trail_In_02 {rendererClasses [], renderers []}
 LWE_Trail_Out_01 {rendererClasses [], renderers []}
 LWE_Trail_Out_02 {rendererClasses [], renderers []}
 LWE_Trail_Start_01 {rendererClasses [], renderers []}
 LWE_Trail_Start_02 {rendererClasses [], renderers []}
 NE_Sparks_Trail_01 {
  emitterUpdateScript [
   &EmitterState(Life Cycle Mode=::NewEnumerator1, Loop Behavior=::NewEnumerator1, Loop Delay=0.35, Loop Duration=0.35, UseLoopDelay=true)
   &SpawnRate(SpawnRate=30, Use Spawn Probability=false)
   &SpawnBurst_Instantaneous(Spawn Count=20, Use Spawn Probability=false)
  ]
  particleSpawnScript [
   &InitializeParticle(
    Lifetime=&UniformRangedFloat(Minimum=0.4, Maximum=0.6)
    Position Offset Coordinate Space=::Local
    Sprite Size=&Multiply_Vector2(A=<0.7 1.5>, B=&Vector2DFromFloat(Value=&Multiply_Float(A=1, B=2)))
    Sprite Size Mode=::NewEnumerator4
    UsePositionOffset=true
    Write Material Random=false
   ) {module=InitializeParticle001}
   &AddVelocity(Velocity=&RandomRangeVector(Minimum=<-600 0 0>, Maximum=<-800 0 0>)) {module=AddVelocity001}
   &BoxLocation(Box Offset=<0 0 75>, Box Size=<500 50 50>, MaskSpawnGroup=false, Module.Override Seed=false, Rotate Box=false, Surface Only=false)
  ]
  particleUpdateScript [
   &ParticleState()
   &Drag(Drag=3, Use Linear Drag=true, Use Rotational Drag=false)
   &ScaleColor(
    Scale Alpha=&Multiply_Float(
     A=&FloatFromCurve(FloatCurve=DataInterface() {propertyValues {Curve {keys [
          {arriveTangent=11.9433, interpMode=RCIM_Cubic, leaveTangent=11.9433, tangentMode=RCTM_User, time=0, value=0}
          <at 0.2 1 RCIM_Cubic>
          <at 1 0 RCIM_Cubic>
         ]}}})
     B=&Multiply_Float(A=1, B=&FNTimeOfDayScalar(Night Scale=0.5, Morning Scale=0.25, Evening Scale=0.25))
    )
    ScaleA=true
    ScaleRGB=true
   ) {module=ScaleColor002}
   &DynamicMaterialParameters(
    Index 0 Param 1=&FloatFromCurve(FloatCurve=DataInterface() {propertyValues {Curve {keys [<at 0 0>,<at 1 1>]}}})
    Index 0 Param 2=1
    Index 0 Param 3=1
    Index 0 Param 4=1
    Module.Param0WriteEnabled=true
    Module.Param1WriteEnabled=true
    Module.Param2WriteEnabled=true
    Module.Param3WriteEnabled=true
   )
   &ScaleSpriteSize(Scale Factor=&Vector2DFromCurve(
     Vector2Curve=DataInterface() {propertyValues {
       XCurve {keys [<at 0 1>,<at 1 0>]}
       YCurve {keys [
         {arriveTangent=-11.1847, interpMode=RCIM_Cubic, leaveTangent=-11.1847, tangentMode=RCTM_User, time=0, value=4}
         {arriveTangent=-3.7, interpMode=RCIM_Cubic, leaveTangent=-3.7, time=0.3897, value=0.99}
         <at 1 0.3>
        ]}
      }}
     ScaleCurve=<1 1>
    ))
   &SolveForcesAndVelocity(Clamp Velocity=false, Limit Acceleration=false)
   &ScaleSpriteSizeBySpeed(Max Scale Factor=<1 4>, Min Scale Factor=<1 2>)
  ]
  rendererClasses [&NiagaraSpriteRendererProperties]
  renderers [{rendererClass=&NiagaraSpriteRendererProperties}]
 }
 NE_Sparks_Trail_02 {
  emitterUpdateScript [
   &EmitterState(Life Cycle Mode=::NewEnumerator1, Loop Behavior=::NewEnumerator1, Loop Delay=1, Loop Duration=0.35, UseLoopDelay=true)
   &SpawnRate(SpawnRate=20, Use Spawn Probability=false)
   &SpawnBurst_Instantaneous(Spawn Count=15, Use Spawn Probability=false)
  ]
  particleSpawnScript [
   &InitializeParticle(
    Lifetime=&UniformRangedFloat(Minimum=0.4, Maximum=0.6)
    Position Offset Coordinate Space=::Local
    Sprite Size=&Multiply_Vector2(A=<0.7 1.5>, B=&Vector2DFromFloat(Value=&Multiply_Float(A=1, B=2)))
    Sprite Size Mode=::NewEnumerator4
    UsePositionOffset=true
    Write Material Random=false
   ) {module=InitializeParticle001}
   &AddVelocity(Velocity=&RandomRangeVector(Minimum=<-600 0 0>, Maximum=<-800 0 0>)) {module=AddVelocity001}
   &BoxLocation(Box Offset=<200 0 75>, Box Size=<250 50 50>, MaskSpawnGroup=false, Module.Override Seed=false, Rotate Box=false, Surface Only=false)
  ]
  particleUpdateScript [
   &ParticleState()
   &Drag(Drag=3, Use Linear Drag=true, Use Rotational Drag=false)
   &ScaleColor(
    Scale Alpha=&Multiply_Float(
     A=&FloatFromCurve(FloatCurve=DataInterface() {propertyValues {Curve {keys [
          {arriveTangent=11.9433, interpMode=RCIM_Cubic, leaveTangent=11.9433, tangentMode=RCTM_User, time=0, value=0}
          <at 0.2 1 RCIM_Cubic>
          <at 1 0 RCIM_Cubic>
         ]}}})
     B=&Multiply_Float(A=1, B=&FNTimeOfDayScalar(Night Scale=0.5, Morning Scale=0.25, Evening Scale=0.25))
    )
    ScaleA=true
    ScaleRGB=true
   ) {module=ScaleColor002}
   &DynamicMaterialParameters(
    Index 0 Param 1=&FloatFromCurve(FloatCurve=DataInterface() {propertyValues {Curve {keys [<at 0 0>,<at 1 1>]}}})
    Index 0 Param 2=1
    Index 0 Param 3=1
    Index 0 Param 4=1
    Module.Param0WriteEnabled=true
    Module.Param1WriteEnabled=true
    Module.Param2WriteEnabled=true
    Module.Param3WriteEnabled=true
   )
   &ScaleSpriteSize(Scale Factor=&Vector2DFromCurve(
     Vector2Curve=DataInterface() {propertyValues {
       XCurve {keys [<at 0 1>,<at 1 0>]}
       YCurve {keys [
         {arriveTangent=-11.1847, interpMode=RCIM_Cubic, leaveTangent=-11.1847, tangentMode=RCTM_User, time=0, value=4}
         {arriveTangent=-3.7, interpMode=RCIM_Cubic, leaveTangent=-3.7, time=0.3897, value=0.99}
         <at 1 0.3>
        ]}
      }}
     ScaleCurve=<1 1>
    ))
   &SolveForcesAndVelocity(Clamp Velocity=false, Limit Acceleration=false)
   &ScaleSpriteSizeBySpeed(Max Scale Factor=<1 4>, Min Scale Factor=<1 2>)
  ]
  rendererClasses [&NiagaraSpriteRendererProperties]
  renderers [{rendererClass=&NiagaraSpriteRendererProperties}]
 }
}
summary {
 systemName=NS_Emote_ShirtTilapia_Trail
 emitters [
  {emitterName=NE_Sparks_Trail_01, rendererClasses [&NiagaraSpriteRendererProperties]}
  {emitterName=LWE_Trail_In_01, rendererClasses []}
  {emitterName=LWE_Trail_In_02, rendererClasses []}
  {emitterName=NE_Sparks_Trail_02, rendererClasses [&NiagaraSpriteRendererProperties]}
  {emitterName=LWE_Trail_Out_01, rendererClasses []}
  {emitterName=LWE_Trail_Out_02, rendererClasses []}
  {emitterName=LWE_Trail_Start_01, rendererClasses []}
  {emitterName=LWE_Trail_Start_02, rendererClasses []}
 ]
 userVariables [
  User.bIsFrontEnd {type=NiagaraBool}
  User.bIsFrontEndPreview {type=NiagaraBool}
  User.RemoveAirflow {type=NiagaraBool}
 ]
}
