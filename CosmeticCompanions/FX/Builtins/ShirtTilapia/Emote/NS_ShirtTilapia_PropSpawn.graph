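# Niagara VFX system for the prop spawn in the ShirtTilapia cosmetic companion emote.
# Three emitters, SparklesLWE, AirglowLWE and SmokeLWE, apparently lightweight by their suffix, supply sparkles, airglow and smoke, with no renderers listed.
# The system uses SystemState plus a spawn-time Set Variables module, alongside a MakeCustomIntFromBool dynamic input.
# User parameters expose front-end and front-end preview flags and a normalized notify progress float.
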NiagaraSystem /CosmeticCompanions/FX/Builtins/ShirtTilapia/Emote/NS_ShirtTilapia_PropSpawn
dependencies {
 usedDataInterfaces []
 usedDynamicInputs [&MakeCustomIntFromBool]
 usedModules [
  &self:SystemSpawnScript.SystemScriptSource.SystemScriptGraph.NiagaraNodeAssignment_2.SetVariables_#1
  &SystemState
 ]
}
emitters {AirglowLWE {rendererClasses [], renderers []}, SmokeLWE {rendererClasses [], renderers []}, SparklesLWE {rendererClasses [], renderers []}}
summary {
 systemName=NS_ShirtTilapia_PropSpawn
 emitters [{emitterName=SparklesLWE, rendererClasses []}, {emitterName=AirglowLWE, rendererClasses []}, {emitterName=SmokeLWE, rendererClasses []}]
 userVariables [
  User.bIsFrontEnd {type=NiagaraBool}
  User.bIsFrontEndPreview {type=NiagaraBool}
  User.NormalizedNotifyProgress {type=NiagaraFloat, defaultValue=0, description=""}
 ]
}
